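; UI director that shows a reticle status scene while the viewed player pawn holds the BarnTowel weapon.
; It follows the owning controller's view target changes, rebinding the weapon-equipped delegate on the new pawn and immediately checking its current weapon.
; A cast to B_Weapon_BarnTowel decides visibility, and a delta gate adds or removes the scene only when that state changes.
; EndPlay unbinds and clears the pawn.
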(event EventBeginPlay
  ; Begin/EndPlay
  (bind _asfort_player_controller (Utilities|Casting|CastToFortPlayerController (DynamicUIDirector|GetOwningLocalPlayerController)))
  (bind _output_get (Variables|Default|SetBoundPlayerController _asfort_player_controller))
  (EventDispatchers|BindEventtoOnViewTargetChanged _output_get (EventDispatchers|CreateEvent :!event OnViewTargetChanged))
  (CallFunction|SetPlayerPawn (Pawn|GetViewTarget _asfort_player_controller)))

(event EventEndPlay
  (params (EndPlayReason :type EEndPlayReason%20Enum :default "Destroyed"))
  (bind _boundplayercontroller (Variables|Default|GetBoundPlayerController))
  ; Begin/EndPlay
  (Utilities|IsValid _boundplayercontroller
    (:Is%20Valid
      (EventDispatchers|UnbindEventfromOnViewTargetChanged _boundplayercontroller (EventDispatchers|CreateEvent :!event OnViewTargetChanged)))
    (:Is%20Not%20Valid
      _))
  (CallFunction|SetPlayerPawn 0))

(fn OnViewTargetChanged
  (params (PlayerController :type Fort%20Player%20Controller%20Object%20Reference) (OldViewTarget :type Actor%20Object%20Reference) (NewViewTarget :type Actor%20Object%20Reference))
  (CallFunction|SetPlayerPawn NewViewTarget))

(fn OnWeaponEquipped
  (params (NewWeapon :type Fort%20Weapon%20Object%20Reference) (PrevWeapon :type Fort%20Weapon%20Object%20Reference))
  (Utilities|Casting|CastToB_Weapon_BarnTowel NewWeapon
    (:then
      (CallFunction|ShowReticleStatusScene true))
    (:CastFailed
      (CallFunction|ShowReticleStatusScene false))))

(fn SetPlayerPawn
  (params (NewViewTarget :type Actor%20Object%20Reference))
  (bind _outputdelegate (EventDispatchers|CreateEvent :!event OnWeaponEquipped))
  (bind _boundviewtarget (Variables|Default|GetBoundViewTarget))
  (Utilities|IsValid _boundviewtarget
    (:Is%20Valid
      (FortPawn|UnbindEventfromOnWeaponEquippedDelegate _boundviewtarget _outputdelegate)
      (Variables|Default|SetBoundViewTarget 0))
    (:Is%20Not%20Valid
      _))
  (bind _asfort_player_pawn (Utilities|Casting|CastToFortPlayerPawn (Variables|GetNewViewTarget)))
  (bind _output_get (Variables|Default|SetBoundViewTarget _asfort_player_pawn))
  (FortPawn|BindEventtoOnWeaponEquippedDelegate _output_get _outputdelegate)
  (CallFunction|OnWeaponEquipped (Weapon|GetCurrentWeapon _output_get)))

(fn ShowReticleStatusScene
  (params (DoShow :type Boolean))
  (bind _reticlestatusscene (Variables|Default|GetReticleStatusScene))
  (bind _updated_variable (Utilities|FlowControl|DeltaGateWildcard (Variables|Default|GetIsReticleStatusSceneShowing) DoShow)
    (:Success
      (if _updated_variable
        (DynamicUIDirector|AddScene _reticlestatusscene)
        (else
          (DynamicUIDirector|RemoveScene _reticlestatusscene))))))

(fn ConstructionScript)
